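# Sound cue for the Sweet Toss emote's foley of a box moving out.
# A random node picks one of two equally weighted wave variations, and a modulator varies the result on each play.
# It uses emote foley sound class, attenuation and concurrency settings, keeps playing when silent, and has a submix send to the music audio analysis submix.
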SoundCue /BRCosmetics/Sounds/Emotes/SweetToss/EmoteFoley_SweetToss_Box_Move_Out_Cue
root=Modulator_0
nodes [
 {id=Modulator_0, children [Random_0], properties {volumeMax=1.1}}
 {id=Random_0, children [WavePlayer_1,WavePlayer_2], properties {weights [1,1]}}
 {id=WavePlayer_1, properties {soundWaveAssetPtr=&/Emote_Sweet_Toss_Box_Move_Out_01}}
 {id=WavePlayer_2, properties {soundWaveAssetPtr=&/Emote_Sweet_Toss_Box_Move_Out_02}}
]
properties {
 attenuationSettings=&Fort_Emotes_Athena_Att
 audioPropertiesSheet=/BRCosmetics/Sounds/Settings/PropertySheets/Emotes/APS_Cosmetic_Emote_Foley
 maxDistance=1600
 soundClassObject=/Game/Sounds/FortSoundClassesAndModes/Fort_Foley_Emote
 virtualizationMode=PlayWhenSilent
 volumeMultiplier=1.2
 concurrencyOverrides {maxCount=3}
 concurrencySet [/Game/Sounds/Concurrency_Rules/Athena_EmoteFoley_Conc]
 soundSubmixSends [{minSendLevel=0, soundSubmix=&Fortnite_AudioAnalysis_Music}]
}
