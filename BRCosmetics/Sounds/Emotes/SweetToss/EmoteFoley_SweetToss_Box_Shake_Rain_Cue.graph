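# Sound cue for the Sweet Toss emote foley, playing a box shake rain sound.
# An equal-weight random node picks one of three waves, with the first given an extra pitch and volume boost, under a root modulator that varies both.
# It uses emote foley attenuation, sound class and concurrency, keeps playing when silent, and sends to a music audio-analysis submix.
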SoundCue /BRCosmetics/Sounds/Emotes/SweetToss/EmoteFoley_SweetToss_Box_Shake_Rain_Cue
root=Modulator_0
nodes [
 {id=Modulator_0, children [Random_1], properties {pitchMax=1.1, volumeMax=1.1}}
 {id=Modulator_1, children [WavePlayer_0], properties {pitchMin=1.05, volumeMax=1.1, volumeMin=1.1}}
 {id=Random_1, children [Modulator_1,WavePlayer_5,WavePlayer_6], properties {weights [1,1,1]}}
 {id=WavePlayer_0, properties {soundWaveAssetPtr=&/Emote_Sweet_Toss_Shake_Rain_01}}
 {id=WavePlayer_5, properties {soundWaveAssetPtr=&/Emote_Sweet_Toss_Shake_Rain_02}}
 {id=WavePlayer_6, properties {soundWaveAssetPtr=&/Emote_Sweet_Toss_Shake_Rain_03}}
]
properties {
 attenuationSettings=&Fort_Emotes_Athena_Att
 audioPropertiesSheet=/BRCosmetics/Sounds/Settings/PropertySheets/Emotes/APS_Cosmetic_Emote_Foley
 maxDistance=1600
 soundClassObject=/Game/Sounds/FortSoundClassesAndModes/Fort_Foley_Emote
 virtualizationMode=PlayWhenSilent
 volumeMultiplier=1.2
 concurrencyOverrides {maxCount=3}
 concurrencySet [/Game/Sounds/Concurrency_Rules/Athena_EmoteFoley_Conc]
 soundSubmixSends [{minSendLevel=0, soundSubmix=&Fortnite_AudioAnalysis_Music}]
}
